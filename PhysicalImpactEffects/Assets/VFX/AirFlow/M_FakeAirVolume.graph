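# Translucent unlit Niagara sprite material for a fake air volume in impact airflow effects, with depth testing disabled.
# A vector parameter, transformed from local to view space, is scaled by particle alpha, a strength scalar and two rerouted masks.
# The graph declares a smoothstepped radial UV mask and a scene-minus-pixel depth remap, whose symmetric range comes from a dynamic parameter.
# Particle direction and speed nodes are unconnected.
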Material /PhysicalImpactEffects/Assets/VFX/AirFlow/M_FakeAirVolume
graph {
 properties {
  bAutomaticallySetUsageInEditor=false
  bDisableDepthTest=true
  bUsedWithNiagaraSprites=true
  blendMode=BLEND_Translucent
  shadingModel=MSM_Unlit
  usedShadingModels=Unlit
 }
 nodes [
  DynamicParameter() {defaultValue=<rgba 1 0 1 500>, paramNames [Param1, SmoothStepMin, SmoothStepMiax, DepthRange]}
  n1 = VectorParameter() {defaultValue=<rgba 0 1 0 0>}
  n2 = ParticleColor()
  n3 = Multiply(A=NamedRerouteUsage(), B=NamedRerouteUsage())
  Multiply(
   A=Multiply(A=Transform(None=@n1.RGB) {transformSourceType=TRANSFORMSOURCE_Local, transformType=TRANSFORM_View}, B=Multiply(A=@n2.A, B=@n3))
   B=ScalarParameter() {defaultValue=0.1}
  )
  n4 = SceneDepth()
  n5 = PixelDepth()
  n6 = DynamicParameter() {defaultValue=<rgba 1 0 1 500>, paramNames [Param1, SmoothStepMin, SmoothStepMiax, DepthRange]}
  NamedRerouteDeclaration(None=MaterialFunctionCall(Input=Subtract(A=@n4.R, B=@n5.R), Input Low=Multiply(B=@n6.DepthRange) {constA=-1}, Input High=@n6.DepthRange) {materialFunction=&RemapValue_0-1}.Result) {name=DepthRemap, nodeColor=<rgba 1 0 1 1>}
  n7 = TextureCoordinate()
  n8 = Constant() {r=0.5}
  NamedRerouteDeclaration(None=SmoothStep(Value=SphereMask(A=@n7, B=@n8, Radius=@n8) {hardnessPercent=0})) {name=SimpleSphereMask, nodeColor=<rgba 1 1 1 1>}
  ParticleDirection()
  ParticleSpeed()
 ]
}
