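# Translucent unlit Niagara sprite material for the AirFlow VFX that acts as a motion blur layer, drawn without depth testing and writing translucent velocity.
# A Z vector scaled by particle alpha and a soft radial UV mask drives two transformed offsets.
# A PreviousFrameSwitch applies the Intensity-scaled vector only to the previous frame, faking motion for blur.
# The camera-transformed vector, scaled by Ref plus temporal dither, forms a 2D refraction-style offset.
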Material /PhysicalImpactEffects/Assets/VFX/AirFlow/M_MotionBlurLayer
graph {
 properties {
  bAutomaticallySetUsageInEditor=false
  bDisableDepthTest=true
  bOutputTranslucentVelocity=true
  bUsedWithNiagaraSprites=true
  blendMode=BLEND_Translucent
  refractionMethod=RM_2DOffset
  shadingModel=MSM_Unlit
  usedShadingModels=Unlit
 }
 nodes [
  n1 = Constant() {r=0.5}
  n2 = Multiply(
   A=Multiply(A=Constant3Vector() {constant=<rgba 0 0 1 0>}, B=ParticleColor().A)
   B=SphereMask(A=TextureCoordinate(), B=@n1, Radius=@n1) {hardnessPercent=0}
  )
  n3 = MaterialFunctionCall() {materialFunction=&DitherTemporalAA}
  n4 = ScalarParameter() {defaultValue=0.3, parameterName=RefDither}
  Multiply(A=Transform(None=@n2) {transformType=TRANSFORM_Camera}, B=Add(A=Multiply(A=@n3.Result, B=@n4), B=ScalarParameter() {parameterName=Ref}))
  n5 = ScalarParameter() {defaultValue=100, parameterName=Intensity}
  Transform(None=PreviousFrameSwitch(Current Frame=Constant(), Previous Frame=Multiply(A=@n5, B=@n2)))
  VectorParameter()
 ]
}
